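# Translucent UI material for the MistRaven Tendril Vacuum HUD reticle, drawing top and bottom brackets plus an arc fill bar.
# MSDF bracket textures blend to alternate shapes and shift position with ValidTarget.
# A subtracted-circle SDF ring, cut by rotated radial masks, shows Fill, with color driven by IsAscended and Flash.
# Antialiased thresholds add shadows and backing, named reroutes link the layers, and Scale resizes the parts.
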Material /MistRaven/HUD/TendrilVaccum/Materials/M_UI_MistRaven_TendrilVacuum_Reticle_New
graph {
 properties {bAutomaticallySetUsageInEditor=false, blendMode=BLEND_Translucent, materialDomain=MD_UI, usedShadingModels=Default Lit}
 nodes [
  n1 = NamedRerouteUsage()
  n2 = LinearInterpolate(A=Constant() {r=0.3}, Alpha=ScalarParameter() {parameterName=ValidTarget}) {constA=0.2}
  n3 = Multiply(A=NamedRerouteUsage(), B=ScalarParameter() {defaultValue=1, parameterName=FillBarOpacity})
  n4 = NamedRerouteUsage()
  Multiply(A=Add(A=Multiply(A=@n1, B=@n2), B=Max(A=@n3, B=@n4)), B=ScalarParameter() {defaultValue=0.85, parameterName=MasterOpacity})
  n5 = Constant() {r=0.2}
  n6 = MaterialFunctionCall(DistanceField=NamedRerouteUsage(), Threshold=@n5) {materialFunction=&MF_UI_AntiAliasing}
  NamedRerouteDeclaration(None=@n6.Outer) {name=AA'dBrackets, nodeColor=<rgba 0 0 0 1>}
  n7 = ComponentMask(None=MaterialFunctionCall!(Vector or UVs) {materialFunction=/FortUILibrary/MaterialFunctions/Basics/MF_UI_CheaperVectorToRadial}.Radial Coordinates) {r=true}
  n8 = Subtract(A=@n7, B=ScalarParameter() {defaultValue=0.4, parameterName=LightGradientAngle}) {constB=0.4}
  n9 = Subtract(A=OneMinus(None=@n7), B=ScalarParameter() {defaultValue=0.4, parameterName=LightGradientAngle}) {constB=0.4}
  n10 = Multiply(A=Multiply(A=Saturate(None=@n8), B=Saturate(None=@n9)), B=ScalarParameter() {defaultValue=60, parameterName=Brightness}) {constB=90}
  n11 = ScalarParameter() {defaultValue=5, parameterName=LightGradientFalloff}
  NamedRerouteDeclaration(None=Saturate(None=Power(Base=@n10, Exp=@n11) {constExponent=1.5})) {name=FillBarLightGradient, nodeColor=<rgba 0 1 0.6016 1>}
  n12 = ComponentMask(None=MaterialFunctionCall(UVs=TextureCoordinate(), Texture Scale=Constant() {r=0.1}) {materialFunction=&ScaleUVsByCenter}.UVs) {r=true}
  n13 = OneMinus(None=@n12)
  NamedRerouteDeclaration(None=Saturate(None=Multiply(A=@n12, B=@n13))) {name=LightGradient, nodeColor=<rgba 1 0 0.2344 1>}
  NamedRerouteDeclaration(None=ScalarParameter() {defaultValue=0.02, parameterName=FillBarLength}) {name=FillLengthOffset, nodeColor=<rgba 0 0 0 1>}
  n14 = LinearInterpolate(
   A=LinearInterpolate(
    A=VectorParameter() {parameterName=NormalFillColor, defaultValue=<rgba 1 0.0048 0.0602 1>}.RGB
    B=VectorParameter() {parameterName=AscendedFillColor, defaultValue=<rgba 0.5029 0 1 1>}.RGB
    Alpha=ScalarParameter() {parameterName=IsAscended}
   )
   Alpha=ScalarParameter() {parameterName=Flash}
  )
  n15 = NamedRerouteUsage()
  n16 = TextureCoordinate()
  MaterialExpressionLinearInterpolate_4 = LinearInterpolate(B=ScalarParameter() {defaultValue=0.08, parameterName=BracketOffset}, Alpha=ScalarParameter() {parameterName=ValidTarget})
  n17 = Convert(None=@MaterialExpressionLinearInterpolate_4) {
   convertInputs [{
     type=Scalar
     expressionInput {
      expression=&self:MaterialExpressionLinearInterpolate_4
      inputName=None
      mask=0
      maskA=0
      maskB=0
      maskG=0
      maskR=0
      outputIndex=0
     }
    }]
   convertOutputs [{type=Vector2}]
  }
  n18 = Constant2Vector() {g=1.03, r=1.2}
  n19 = Power(Base=ScalarParameter() {defaultValue=0.8, parameterName=Scale}) {constExponent=0.5}
  n20 = MaterialFunctionCall(UVs=Add(A=Clamp(None=@n16) {maxDefault=0.999}, B=@n17), Texture Scale=AppendVector(A=Multiply(A=@n18.R, B=@n19), B=@n18.G)) {materialFunction=&ScaleUVsByCenter}
  n21 = ScalarParameter() {defaultValue=0.8, parameterName=Scale}
  n22 = Subtract(A=ScalarParameter() {defaultValue=0.974, parameterName=FillBarRadius}, B=Multiply(A=OneMinus(None=@n21)) {constB=0.25})
  n23 = Multiply(A=ScalarParameter() {defaultValue=0.13, parameterName=FillBarThickness}, B=ScalarParameter() {defaultValue=0.8, parameterName=Scale})
  n24 = MaterialFunctionCall!(
   DistanceFieldA=MaterialFunctionCall!(UVs=@n20.UVs, Radius=@n22) {materialFunction=/FortUILibrary/MaterialFunctions/Shapes/SDFs/MF_SDF_Circle}.SDF
   DistanceFieldB=MaterialFunctionCall(UVs=@n20.UVs, Radius=Subtract(A=@n22, B=@n23)) {materialFunction=&MF_SDF_Circle}.SDF
   BlendAmount
  ) {materialFunction=/FortUILibrary/MaterialFunctions/Shapes/Modifiers/MF_UI_SDF_Subtraction}
  n25 = MaterialFunctionCall(DistanceField=@n24.Exact) {materialFunction=&MF_UI_AntiAliasing}
  n26 = Max(A=Multiply(A=@n25.Inner) {constB=0}, B=NamedRerouteUsage())
  NamedRerouteDeclaration(None=Multiply(A=Add(A=@n14, B=@n15), B=@n26)) {name=FillBar, nodeColor=<rgba 0 0 0 1>}
  NamedRerouteDeclaration(None=@n26) {name=FillBarOpacity, nodeColor=<rgba 0 0 0 1>}
  n27 = ScalarParameter() {parameterName=TopBracketOffsetX}
  n28 = Subtract(
   A=ScalarParameter() {defaultValue=0.329, parameterName=TopBracketOffsetY}
   B=LinearInterpolate(B=ScalarParameter() {defaultValue=0.08, parameterName=BracketOffset}, Alpha=ScalarParameter() {parameterName=ValidTarget})
  )
  n29 = MaterialFunctionCall(
   UVs=Add(A=TextureCoordinate(), B=AppendVector(A=@n27, B=@n28))
   Texture Scale=Multiply(A=ScalarParameter() {defaultValue=0.8, parameterName=Scale}, B=ScalarParameter() {defaultValue=1, parameterName=TopBracketScale})
  ) {materialFunction=&ScaleUVsByCenter}
  n30 = MaterialFunctionCall!(MSDF RGB=TextureSample(UVs=@n29.UVs) {texture=/MistRaven/HUD/TendrilVaccum/Textures/T_UI_MSDF_MistRaven_TendrilVacuum_TopBracket}.RGB) {materialFunction=/FortUILibrary/MaterialFunctions/Shapes/Modifiers/MF_UI_MSDFtoSDF}
  n31 = MaterialFunctionCall(MSDF RGB=TextureSample(UVs=@n29.UVs) {texture=/MistRaven/HUD/TendrilVaccum/Textures/T_UI_MSDF_MistRaven_TendrilVacuum_TopBracket_New}.RGB) {materialFunction=&MF_UI_MSDFtoSDF}
  n32 = ScalarParameter() {parameterName=ValidTarget, sliderMax=1}
  n33 = ScalarParameter() {parameterName=BottomBracketOffsetX}
  n34 = Add(
   A=ScalarParameter() {defaultValue=-0.34, parameterName=BottomBracketOffsetY}
   B=LinearInterpolate(B=ScalarParameter() {defaultValue=0.08, parameterName=BracketOffset}, Alpha=ScalarParameter() {parameterName=ValidTarget})
  )
  n35 = MaterialFunctionCall(
   UVs=Add(A=TextureCoordinate(), B=AppendVector(A=@n33, B=@n34))
   Texture Scale=Multiply(A=ScalarParameter() {defaultValue=0.8, parameterName=Scale}, B=ScalarParameter() {defaultValue=1, parameterName=BottomBracketScale})
  ) {materialFunction=&ScaleUVsByCenter}
  n36 = MaterialFunctionCall(MSDF RGB=TextureSample(UVs=@n35.UVs) {texture=/MistRaven/HUD/TendrilVaccum/Textures/T_UI_MSDF_MistRaven_TendrilVacuum_BottomBracket}.RGB) {materialFunction=&MF_UI_MSDFtoSDF}
  n37 = MaterialFunctionCall(MSDF RGB=TextureSample(UVs=@n35.UVs) {texture=/MistRaven/HUD/TendrilVaccum/Textures/T_UI_MSDF_MistRaven_TendrilVacuum_BottomBracket_New}.RGB) {materialFunction=&MF_UI_MSDFtoSDF}
  n38 = ScalarParameter() {parameterName=ValidTarget}
  NamedRerouteDeclaration(None=Max(A=LinearInterpolate(A=@n30.SDF, B=@n31.SDF, Alpha=@n32), B=LinearInterpolate(A=@n36.SDF, B=@n37.SDF, Alpha=@n38))) {name=Brackets, nodeColor=<rgba 1 0 0.5859 1>}
  n39 = NamedRerouteUsage()
  n40 = Add(A=@n5) {constB=-0.1}
  NamedRerouteDeclaration(None=LinearInterpolate(
    A=@n6.Outer
    B=MaterialFunctionCall(DistanceField=@n39, Threshold=@n40) {materialFunction=&MF_UI_AntiAliasing}.Outer
    Alpha=ScalarParameter() {defaultValue=0.5, parameterName=ShadowOpacity}
   )) {name=BracketShadows, nodeColor=<rgba 0 0 0 1>}
  MaterialExpressionLinearInterpolate_11 = LinearInterpolate(B=ScalarParameter() {defaultValue=0.08, parameterName=BracketOffset}, Alpha=ScalarParameter() {parameterName=ValidTarget})
  n41 = Convert(None=@MaterialExpressionLinearInterpolate_11) {
   convertInputs [{
     type=Scalar
     expressionInput {
      expression=&self:MaterialExpressionLinearInterpolate_11
      inputName=None
      mask=0
      maskA=0
      maskB=0
      maskG=0
      maskR=0
      outputIndex=0
     }
    }]
   convertOutputs [{type=Vector2}]
  }
  n42 = Multiply(A=ScalarParameter() {defaultValue=0.8, parameterName=Scale}, B=ScalarParameter() {defaultValue=1.05, parameterName=ShadowMaskScale})
  n43 = MaterialFunctionCall(UVs=Add(A=TextureCoordinate(), B=@n41), Texture Scale=Multiply(A=Constant2Vector() {g=1, r=1}, B=@n42)) {materialFunction=&ScaleUVsByCenter}
  n44 = OneMinus(None=@n42)
  n45 = NamedRerouteUsage()
  n46 = Add(
   A=ScalarParameter() {defaultValue=0.645, parameterName=ProgressBarRotation}
   B=Add(A=Multiply(A=@n44) {constB=0.103}, B=Multiply(A=@n45) {constB=-0.36})
  )
  n47 = MaterialFunctionCall(Vector or UVs=Subtract(A=MaterialFunctionCall(UVs=@n43.UVs, "Rotation Angle (0-1)"=@n46) {materialFunction=&CustomRotator}.Rotated Values) {constB=0.5}) {materialFunction=&VectorToRadialValue}
  n48 = Constant2Vector() {g=1}
  n49 = Power(Base=Multiply(A=ScalarParameter() {defaultValue=0.8, parameterName=Scale}, B=ScalarParameter() {defaultValue=1.05, parameterName=ShadowMaskScale})) {constExponent=0.9}
  n50 = Constant2Vector() {g=0.205, r=0.02}
  n51 = NamedRerouteUsage()
  n52 = NamedRerouteUsage()
  n53 = MaterialFunctionCall(
   Input=Constant() {r=1}
   Input Low=@n48.R
   Input High=@n48.R
   Target Low=Multiply(A=Subtract(A=@n50.R, B=@n52), B=@n49)
   Target High=Multiply(A=@n49, B=Add(A=@n50.G, B=@n51))
  ) {materialFunction=&RemapValueRange}
  NamedRerouteDeclaration(None=Floor(None=Add(A=@n53.Result, B=@n47.Radial Coordinates))) {name=RadialShadowMask, nodeColor=<rgba 1 0.0938 0 1>}
  n54 = ScalarParameter() {defaultValue=1, parameterName=Fill, sliderMax=1}
  n55 = MaterialFunctionCall!(A=ScalarParameter() {defaultValue=1, parameterName=Fill, sliderMax=1}, B=Constant() {r=1}) {materialFunction=/FortUILibrary/MaterialFunctions/Basics/MF_UI_BooleanOperations}
  n56 = Constant2Vector() {g=1}
  n57 = Constant2Vector() {g=0.205, r=0.02}
  n58 = Subtract(A=@n57.R, B=NamedRerouteUsage())
  n59 = ScalarParameter() {defaultValue=0.8, parameterName=Scale}
  n60 = Add(A=@n57.G, B=NamedRerouteUsage())
  n61 = Power(Base=@n59) {constExponent=0.9}
  MaterialExpressionLinearInterpolate_5 = LinearInterpolate(B=ScalarParameter() {defaultValue=0.08, parameterName=BracketOffset}, Alpha=ScalarParameter() {parameterName=ValidTarget})
  n62 = Convert(None=@MaterialExpressionLinearInterpolate_5) {
   convertInputs [{
     type=Scalar
     expressionInput {
      expression=&self:MaterialExpressionLinearInterpolate_5
      inputName=None
      mask=0
      maskA=0
      maskB=0
      maskG=0
      maskR=0
      outputIndex=0
     }
    }]
   convertOutputs [{type=Vector2}]
  }
  n63 = ScalarParameter() {defaultValue=0.8, parameterName=Scale}
  n64 = MaterialFunctionCall(UVs=Add(A=TextureCoordinate(), B=@n62), Texture Scale=Multiply(A=Constant2Vector() {g=1, r=1}, B=@n63)) {materialFunction=&ScaleUVsByCenter}
  n65 = OneMinus(None=@n63)
  n66 = NamedRerouteUsage()
  n67 = Add(
   A=ScalarParameter() {defaultValue=0.645, parameterName=ProgressBarRotation}
   B=Add(A=Multiply(A=@n65) {constB=0.103}, B=Multiply(A=@n66) {constB=-0.36})
  )
  n68 = MaterialFunctionCall(Vector or UVs=Subtract(A=MaterialFunctionCall(UVs=@n64.UVs, "Rotation Angle (0-1)"=@n67) {materialFunction=&CustomRotator}.Rotated Values) {constB=0.5}) {materialFunction=&VectorToRadialValue}
  n69 = Add(
   A=MaterialFunctionCall(
    Input=LinearInterpolate(A=@n54, Alpha=@n55.IsGreater)
    Input Low=@n56.R
    Input High=@n56.R
    Target Low=Multiply(A=@n58, B=@n59)
    Target High=Multiply(A=@n60, B=@n61)
   ) {materialFunction=&RemapValueRange}.Result
   B=@n68.Radial Coordinates
  )
  NamedRerouteDeclaration(None=Multiply(A=Floor(None=@n69), B=@n25.Inner)) {name=RadialFillMask, nodeColor=<rgba 0 0.9062 1 1>}
  n70 = MaterialFunctionCall(DistanceField=@n24.Exact, Threshold=ScalarParameter() {defaultValue=0.02, parameterName=ShadowThreshold}) {materialFunction=&MF_UI_AntiAliasing}
  n71 = Multiply(A=@n70.Inner, B=NamedRerouteUsage())
  n72 = ScalarParameter() {defaultValue=0.5, parameterName=FillBarBackingOpacity}
  NamedRerouteDeclaration(None=Max(A=Multiply(A=@n70.Inner) {constB=0}, B=Multiply(A=@n71, B=@n72) {constB=0.5})) {name=FillBarShadow, nodeColor=<rgba 1 0.7734 0 1>}
  n73 = Constant2Vector() {g=1}
  n74 = Constant2Vector() {g=0.205, r=0.02}
  n75 = NamedRerouteUsage()
  n76 = Power(Base=ScalarParameter() {defaultValue=0.8, parameterName=Scale}) {constExponent=0.9}
  n77 = NamedRerouteUsage()
  n78 = MaterialFunctionCall(
   Input=Constant() {r=1}
   Input Low=@n73.R
   Input High=@n73.R
   Target Low=Multiply(A=Subtract(A=@n74.R, B=@n77), B=@n76)
   Target High=Multiply(A=Add(A=@n74.G, B=@n75), B=@n76)
  ) {materialFunction=&RemapValueRange}
  NamedRerouteDeclaration(None=Floor(None=Add(A=@n78.Result, B=@n68.Radial Coordinates))) {name=FillBackgroundMask, nodeColor=<rgba 0.4531 0 1 1>}
  NamedRerouteUsage()
  Reroute(None=Add(A=NamedRerouteUsage(), B=NamedRerouteUsage()))
 ]
}
