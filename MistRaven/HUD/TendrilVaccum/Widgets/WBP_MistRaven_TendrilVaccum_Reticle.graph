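# Weapon reticle extension widget for the MistRaven Tendril Vacuum, showing ammo, reload progress and target validity for the ascended flamethrower.
# On Construct it caches the reticle's dynamic material, sets its IsAscended parameter and binds weapon ammo and reload delegates.
# Ammo ratio drives Fill, while a looping timer animates Fill and Flash during reloads.
# Async gameplay-tag waits on the PlayerController handle attack, target, ascended-active and wraith-form changes, playing target and fade animations.
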WidgetBlueprint /MistRaven/HUD/TendrilVaccum/Widgets/WBP_MistRaven_TendrilVaccum_Reticle
animations [
 Anim_EquipNoTarget {
  binding_count=1
  length_seconds=0.25
  object=&self:Anim_EquipNoTarget
  sequence {bindings [Image_Reticle {
     id=#1
     possessed_class=FortLazyImage
     tracks [{
       type=MovieSceneColorTrack
       display_name=Color and Opacity
       sections [{
         type=MovieSceneColorSection
         end_frame=5
         start_frame=0
         channels [
          Color.R {type=MovieSceneScriptingFloatChannel, const [1,1]}
          Color.G {type=MovieSceneScriptingFloatChannel, const [1,1]}
          Color.B {type=MovieSceneScriptingFloatChannel, const [1,1]}
          Color.A {type=MovieSceneScriptingFloatChannel, keys [[0,0],[5,1]]}
         ]
        }]
      }]
    }]}
 }
 Anim_TargetFound {
  binding_count=1
  length_seconds=0.15
  object=&self:Anim_TargetFound
  sequence {bindings [Image_Reticle {
     id=#2
     possessed_class=FortLazyImage
     tracks [{
       type=MovieSceneWidgetMaterialTrack
       sections [{
         type=MovieSceneParameterSection
         end_frame=3
         start_frame=0
         channels [ValidTarget {type=MovieSceneScriptingFloatChannel, keys [[0,0],[3,1]]}]
        }]
      }]
    }]}
 }
 Anim_HideBar {
  binding_count=1
  length_seconds=0.35
  object=&self:Anim_HideBar
  sequence {bindings [Image_Reticle {
     id=#3
     possessed_class=FortLazyImage
     tracks [{
       type=MovieSceneColorTrack
       display_name=Color and Opacity
       sections [{
         type=MovieSceneColorSection
         end_frame=7
         start_frame=0
         channels [
          Color.R {type=MovieSceneScriptingFloatChannel, default=1}
          Color.G {type=MovieSceneScriptingFloatChannel, default=1}
          Color.B {type=MovieSceneScriptingFloatChannel, default=1}
          Color.A {type=MovieSceneScriptingFloatChannel, keys [[0,1],[7,0]]}
         ]
        }]
      }]
    }]}
 }
 Anim_ShowBar {
  binding_count=1
  length_seconds=0.35
  object=&self:Anim_ShowBar
  sequence {bindings [Image_Reticle {
     id=#4
     possessed_class=FortLazyImage
     tracks [{
       type=MovieSceneColorTrack
       display_name=Color and Opacity
       sections [{
         type=MovieSceneColorSection
         end_frame=7
         start_frame=0
         channels [
          Color.R {type=MovieSceneScriptingFloatChannel, default=1}
          Color.G {type=MovieSceneScriptingFloatChannel, default=1}
          Color.B {type=MovieSceneScriptingFloatChannel, default=1}
          Color.A {type=MovieSceneScriptingFloatChannel, keys [[0,0],[7,1]]}
         ]
        }]
      }]
    }]}
 }
 Anim_ReloadEnd {
  binding_count=1
  length_seconds=0.15
  object=&self:Anim_ReloadEnd
  sequence {bindings [Image_Reticle {
     id=#5
     possessed_class=FortLazyImage
     tracks [{
       type=MovieSceneWidgetMaterialTrack
       sections [{
         type=MovieSceneParameterSection
         end_frame=3
         start_frame=0
         channels [Flash {type=MovieSceneScriptingFloatChannel, keys [[0,1],[2,0],[3,0]]}]
        }]
      }]
    }]}
 }
 Anim_Reload {
  binding_count=1
  length_seconds=1
  object=&self:Anim_Reload
  sequence {bindings [Image_Reticle {
     id=#6
     possessed_class=FortLazyImage
     tracks [{
       type=MovieSceneWidgetMaterialTrack
       sections [{
         type=MovieSceneParameterSection
         end_frame=20
         start_frame=0
         channels [
          Fill {type=MovieSceneScriptingFloatChannel, keys [[0,0],[20,1]]}
          Flash {type=MovieSceneScriptingFloatChannel, const [1,2]}
         ]
        }]
      }]
    }]}
 }
]
graphs [
 EventGraph {nodes [
   ListenForAttackTags() {is=title K2Node_CustomEvent}
   > n1 = WaitGameplayTagAddToActor!(
    execute:Exec
    TargetActor:Actor Object Reference=Get PlayerController() -> (PlayerController:Player Controller Object Reference)
    Tag:Gameplay Tag Structure=Get AttackGameplayTag() -> (AttackGameplayTag:Gameplay Tag Structure)
    OnlyTriggerOnce:bool=false
   ) -> (then:Exec, AsyncTaskProxy:Ability Async Wait Gameplay Tag Added Object Reference, Added:Exec) {is=title K2Node_AsyncAction}
   > n2 = WaitGameplayTagRemoveFromActor!(
    execute:Exec
    TargetActor:Actor Object Reference=Get PlayerController()
    Tag:Gameplay Tag Structure=Get AttackGameplayTag()
    OnlyTriggerOnce:bool=false
   ) -> (then:Exec, AsyncTaskProxy:Ability Async Wait Gameplay Tag Removed Object Reference, Removed:Exec) {is=title K2Node_AsyncAction}
   HandleAttack!(execute=@n2.Removed, self, IsAttacking:bool=false) -> (then)
   HandleAttack(execute=@n1.Added, IsAttacking=true)
   OnEquip() {is=title K2Node_CustomEvent}
   > HandleEquipAnimation()
   > ListenForTargetValidTags()
   > ListenForAttackTags() {is=title K2Node_CallFunction}
   > ListenForWraithForm()
   > ListenForAscendedActive()
   ListenForAscendedActive() {is=title K2Node_CustomEvent}
   > n3 = WaitGameplayTagAddToActor(TargetActor=Get PlayerController(), Tag=Get AscendedActiveTag() -> (AscendedActiveTag:Gameplay Tag Structure), OnlyTriggerOnce=false)
   > n4 = WaitGameplayTagRemoveFromActor(TargetActor=Get PlayerController(), Tag=Get AscendedActiveTag(), OnlyTriggerOnce=false)
   HandleAscendedActive!(execute=@n4.Removed, self, IsActive:bool=false) -> (then)
   HandleAscendedActive(execute=@n3.Added, IsActive=true)
   OnReloadEnd() {is=title K2Node_CustomEvent}
   > PlayAnimationForward(InAnimation=Get Anim_ReloadEnd() -> (Anim_ReloadEnd:Widget Animation Object Reference))
   > OnReloadCancel()
   ListenForTargetValidTags() {is=title K2Node_CustomEvent}
   > n5 = WaitGameplayTagAddToActor(
    TargetActor=Get PlayerController()
    Tag=Get NoValidTargetGameplayTag() -> (NoValidTargetGameplayTag:Gameplay Tag Structure)
    OnlyTriggerOnce=false
   )
   > n6 = WaitGameplayTagRemoveFromActor(TargetActor=Get PlayerController(), Tag=Get NoValidTargetGameplayTag(), OnlyTriggerOnce=false)
   HandleValidTarget!(execute=@n6.Removed, self, IsTargetValid:bool=true) -> (then)
   HandleValidTarget(execute=@n5.Added, IsTargetValid=false)
   ListenForWraithForm() {is=title K2Node_CustomEvent}
   > n7 = WaitGameplayTagAddToActor(TargetActor=Get PlayerController(), Tag=Get WraithFormActiveTag() -> (WraithFormActiveTag:Gameplay Tag Structure), OnlyTriggerOnce=false)
   > n8 = WaitGameplayTagRemoveFromActor(TargetActor=Get PlayerController(), Tag=Get WraithFormActiveTag(), OnlyTriggerOnce=false)
   HandleWraithFormState!(execute=@n8.Removed, self, IsWraithForm:"Boolean (by ref)"=Get IsWraithForm() -> (IsWraithForm:bool), SetActive:bool=false) -> (then)
   HandleWraithFormState(execute=@n7.Added, IsWraithForm=Get IsWraithForm(), SetActive=true)
   n9 = OnReload!() -> (OutputDelegate, then, ReloadTime:float=~, ReloadType:EFortWeaponReloadType Enum=ReloadWholeClip) {is=title K2Node_CustomEvent}
   > n10 = Set ReloadDuration(ReloadDuration:float=@n9.ReloadTime) -> (Output_Get:float)
   > n11 = Set Timer by Event(Delegate=Create Event(), Time=@n10, bLooping=false, bMaxOncePerFrame=true)
   > Set ReloadTimerHandle(ReloadTimerHandle:Timer Handle Structure=@n11) -> (Output_Get:Timer Handle Structure)
   > SetScalarParameterValue(self=Get ReticleMaterialInstance() -> (ReticleMaterialInstance:Material Instance Dynamic Object Reference), ParameterName=Fill, Value=0)
   > Set IsReloading(IsReloading:bool=true) -> (Output_Get:bool)
   > n12 = Set Timer by Event(Delegate=Create Event(), Time=Get ReloadFillUpdateTime() -> (ReloadFillUpdateTime:float), bLooping=true)
   > Set ReloadAnimTimerHandle(ReloadAnimTimerHandle:Timer Handle Structure=@n12) -> (Output_Get:Timer Handle Structure)
   OnLocalAmmoChanged!() -> (OutputDelegate, then, LocalCount:int, LocalRemaining:int) {is=title K2Node_CustomEvent}
   > n13 = Branch(Condition=Get IsReloading() -> (IsReloading:bool))
   n14 = Get!(execute:Exec=@n13.else) -> (then:Exec, else:Exec, AsBMistRavenFlameThrower:B Mist Raven Flame Thrower Ascended Object Reference)
   > HandleAmmoUsage!(execute, self, FortWeapon:"Fort Weapon Object Reference (by ref)"=@n14) -> (then)
   Event Construct()
   > Parent: Construct()
   > Set PlayerController(PlayerController:Player Controller Object Reference=GetOwningPlayer()) -> (Output_Get:Player Controller Object Reference)
   > Set PlayerPawn(PlayerPawn:Pawn Object Reference=GetOwningPlayerPawn!(self:User Widget Object Reference) -> (ReturnValue:Pawn Object Reference)) -> (Output_Get:Pawn Object Reference)
   > OnEquip() {is=title K2Node_CallFunction}
   > n15 = GetDynamicMaterial(self=Get Image_Reticle() -> (Image_Reticle:Fort Lazy Image Object Reference))
   > Set ReticleMaterialInstance(ReticleMaterialInstance:Material Instance Dynamic Object Reference=@n15) -> (Output_Get:Material Instance Dynamic Object Reference)
   > n16 = Cast To B_MistRaven_FlameThrowerAscended!(
    execute
    Object=GetEquippedWeapon!(self:Fort Weapon Reticle Extension Widget Base Object Reference) -> (ReturnValue:Fort Weapon Object Reference)
   ) -> (then, CastFailed, AsB Mist Raven Flame Thrower Ascended:B Mist Raven Flame Thrower Ascended Object Reference) {is=title K2Node_DynamicCast}
   > n17 = Set AsBMistRavenFlameThrower(AsBMistRavenFlameThrower:B Mist Raven Flame Thrower Ascended Object Reference=@n16) -> (Output_Get:B Mist Raven Flame Thrower Ascended Object Reference)
   > Assign On Local Ammo Changed!(execute:Exec, self:Fort Weapon Object Reference=@n17, Delegate:"Delegate (by ref)"=Create Event()) -> (then:Exec) {is=title K2Node_AssignDelegate}
   > Assign On Local Reload Started!(
    execute:Exec
    self:Fort Weapon Object Reference=Get AsBMistRavenFlameThrower() -> (AsBMistRavenFlameThrower:B Mist Raven Flame Thrower Ascended Object Reference)
    Delegate:"Delegate (by ref)"=Create Event()
   ) -> (then:Exec) {is=title K2Node_AssignDelegate}
   > Bind Event to On Local Reload Cancelled(self:Fort Weapon Object Reference=Get AsBMistRavenFlameThrower(), Delegate=Create Event()) {is=title K2Node_AddDelegate}
   > n18 = Cast To FortPlayerPawn(Object=Get PlayerPawn() -> (PlayerPawn:Pawn Object Reference))
   > HandleAscended!(
    execute
    self
    IsAscended:bool=HasMatchingGameplayTag!(
     self:Gameplay Tag Asset Interface Interface=@n18
     TagToCheck:Gameplay Tag Structure=Get AscendedEquippedTag() -> (AscendedEquippedTag:Gameplay Tag Structure)
    ) -> (ReturnValue:bool)
   ) -> (then)
   > OnLocalAmmoChanged!(execute, self, LocalCount:int=0, LocalRemaining:int=0) -> (then) {is=title K2Node_CallFunction}
  ]}
 HandleAmmoUsage {nodes [
   n1 = HandleAmmoUsage!() -> (then, FortWeapon:"Fort Weapon Object Reference (by ref)") {is=title K2Node_FunctionEntry}
   n2 = GetLocalMagazineAmmoCount!(self:Fort Weapon Object Reference=@n1) -> (ReturnValue:int=0)
   n3 = GetBulletsPerClip!(self:Fort Weapon Object Reference=@n1) -> (ReturnValue:int=0)
   > SetScalarParameterValue(
    self=Get ReticleMaterialInstance()
    ParameterName=Fill
    Value=float / float(A="To Float (Integer)"(InInt=@n2), B="To Float (Integer)"(InInt=@n3))
   )
  ]}
 HandleAscended {nodes [
   n1 = HandleAscended!() -> (then, IsAscended:bool=~) {is=title K2Node_FunctionEntry}
   > SetScalarParameterValue(self=Get ReticleMaterialInstance(), ParameterName=IsAscended, Value="To Float (Boolean)"(InBool=@n1))
  ]}
 HandleAscendedActive {nodes [
   n1 = HandleAscendedActive!() -> (then, IsActive:bool=~) {is=title K2Node_FunctionEntry}
   > n2 = Branch(Condition=@n1)
   > OnReloadCancel()
   n3 = Get(execute=@n2.else)
   > HandleAmmoUsage(FortWeapon=@n3) {is=title K2Node_CallFunction}
  ]}
 HandleAttack {nodes [
   n1 = HandleAttack!() -> (then, IsAttacking:bool=~) {is=title K2Node_FunctionEntry}
   > n2 = Set IsAttacking(IsAttacking:bool=@n1) -> (Output_Get:bool)
   > Branch(Condition=@n2)
   > Branch(Condition=Get IsTargetValid() -> (IsTargetValid:bool))
   > PlayAnimationForward(InAnimation=Get Anim_TargetFound() -> (Anim_TargetFound:Widget Animation Object Reference))
  ]}
 HandleEquipAnimation {nodes [
   HandleEquipAnimation() {is=title K2Node_FunctionEntry}
   > PlayAnimationForward(InAnimation=Get Anim_EquipNoTarget() -> (Anim_EquipNoTarget:Widget Animation Object Reference))
  ]}
 HandleReloadAnim {nodes [
   HandleReloadAnim() {is=title K2Node_FunctionEntry}
   > n1 = Branch(Condition="float >= float"(A=Get ReloadElapsedTime() -> (ReloadElapsedTime:float), B=Get ReloadDuration() -> (ReloadDuration:float)))
   > OnReloadCancel()
   > SetScalarParameterValue(self=Get ReticleMaterialInstance(), ParameterName=Flash, Value=0)
   > Set ReloadElapsedTime(ReloadElapsedTime:float=0) -> (Output_Get:float)
   SetScalarParameterValue(execute=@n1.else, self=Get ReticleMaterialInstance(), ParameterName=Flash, Value=1)
   > SetScalarParameterValue(self=Get ReticleMaterialInstance(), ParameterName=Fill, Value=float / float(A=Get ReloadElapsedTime(), B=Get ReloadDuration()))
   > Set ReloadElapsedTime(ReloadElapsedTime=float + float(A=Get ReloadElapsedTime(), B=Get ReloadFillUpdateTime()))
  ]}
 HandleValidTarget {nodes [
   n1 = HandleValidTarget!() -> (then, IsTargetValid:bool=~) {is=title K2Node_FunctionEntry}
   > n2 = Delta Gate Wildcard!(execute:Exec, Variable:"Boolean (by ref)"=Get IsTargetValid(), New Value:bool=@n1) -> (Success:Exec, Updated Variable:"Boolean (by ref)", Previous Variable:bool=~) {is=title K2Node_MacroInstance}
   > n3 = Branch(Condition=@n2)
   > PlayAnimationForward(InAnimation=Get Anim_TargetFound())
   PlayAnimationReverse(execute=@n3.else, InAnimation=Get Anim_TargetFound(), PlaybackSpeed=Get NoValidTargetPlayBackSpeed() -> (NoValidTargetPlayBackSpeed:float))
  ]}
 HandleWraithFormState {nodes [
   n1 = HandleWraithFormState!() -> (then, IsWraithForm:"Boolean (by ref)", SetActive:bool=~) {is=title K2Node_FunctionEntry}
   > n2 = Delta Gate Wildcard(Variable=@n1, New Value=@n1.SetActive) -> (Previous Variable=~)
   > n3 = Branch(Condition=@n2)
   > PlayAnimationForward(InAnimation=Get Anim_HideBar() -> (Anim_HideBar:Widget Animation Object Reference))
   PlayAnimationForward(execute=@n3.else, InAnimation=Get Anim_ShowBar() -> (Anim_ShowBar:Widget Animation Object Reference))
  ]}
 OnReloadCancel {nodes [
   OnReloadCancel() {is=title K2Node_FunctionEntry}
   > Clear and Invalidate Timer by Handle(Handle=Get ReloadTimerHandle() -> (ReloadTimerHandle:Timer Handle Structure))
   > Clear and Invalidate Timer by Handle(Handle=Get ReloadAnimTimerHandle() -> (ReloadAnimTimerHandle:Timer Handle Structure))
   > n1 = Get()
   > HandleAmmoUsage(FortWeapon=@n1)
   > SetScalarParameterValue(self=Get ReticleMaterialInstance(), ParameterName=Flash, Value=0)
   > Set ReloadElapsedTime(ReloadElapsedTime=0)
   > Set IsReloading(IsReloading=false)
  ]}
]
variables [
 AsBMistRavenFlameThrower {type=B_MistRaven_FlameThrowerAscended_C, default=null}
 AscendedActiveTag {type=GameplayTag, default {tagName=MistRaven.TendrilDrain.Ascended.Active}}
 AscendedEquippedTag {type=GameplayTag, default {tagName=MistRaven.AllAscendedEquipped}}
 AttackGameplayTag {type=GameplayTag, default {tagName=GameplayCue.Athena.MistRaven.Flamethrower.Active}}
 IsAttacking {type=boolean}
 IsReloading {type=boolean}
 IsTargetValid {type=boolean}
 IsWraithForm {type=boolean}
 NoValidTargetGameplayTag {type=GameplayTag, default {tagName=MistRaven.TendrilDrain.Ascended.NoValidTarget}}
 NoValidTargetPlayBackSpeed {type=number, default=1}
 PlayerController {type=PlayerController, default=null}
 PlayerPawn {type=Pawn, default=null}
 ReloadAnimTimerHandle {type=TimerHandle}
 ReloadDuration {type=number}
 ReloadElapsedTime {type=number}
 ReloadFillUpdateTime {type=number, default=0.02}
 ReloadTimerHandle {type=TimerHandle}
 ReticleMaterialInstance {type=MaterialInstanceDynamic}
 WraithFormActiveTag {type=GameplayTag, default {tagName=MistRaven.GrimStrut.Active}}
]
widgets {
 info {
  inheritedWidgetCount=1
  parentClass=/Script/FortniteUI.FortWeaponReticleExtensionWidgetBase
  rootWidgetClass=&ScaleBox
  widgetCount=3
 }
 widgets [
  ScaleBox_0 {widgetClassPath=&ScaleBox, widget=ScaleBox_0, properties {stretch=UserSpecified, userSpecifiedScale=0.85}}
  Image_Reticle {
   widgetClassPath=&FortLazyImage
   bIsVariable=true
   parent=ScaleBox_0
   widget=Image_Reticle
   properties {
    brush {
     bIsDynamicallyLoaded=false
     drawAs=Box
     imageType=FullColor
     mirroring=NoMirror
     resourceName=None
     resourceObject=/MistRaven/HUD/TendrilVaccum/Materials/M_UI_MistRaven_TendrilVacuum_Reticle_New
     tiling=NoTile
     imageSize=<128 128>
     margin=<ltrb 0.5 0.5 0.5 0.5>
     outlineSettings {
      bUseBrushTransparency=false
      roundingType=HalfHeightRadius
      width=0
      color {colorUseRule=UseColor_Specified, specifiedColor=<rgba 0 0 0 0>}
      cornerRadii=<0 0 0 0>
     }
     tintColor {colorUseRule=UseColor_Specified, specifiedColor=<rgba 1 1 1 1>}
     uVRegion {bIsValid=false, max=<0 0>, min=<0 0>}
    }
    loadingBackgroundBrush {
     bIsDynamicallyLoaded=false
     drawAs=Image
     imageType=FullColor
     mirroring=NoMirror
     resourceName=None
     resourceObject=null
     tiling=NoTile
     imageSize=<32 32>
     margin=<ltrb 0 0 0 0>
     outlineSettings {
      bUseBrushTransparency=false
      roundingType=HalfHeightRadius
      width=0
      color {colorUseRule=UseColor_Specified, specifiedColor=<rgba 0 0 0 0>}
      cornerRadii=<0 0 0 0>
     }
     tintColor {colorUseRule=UseColor_Specified, specifiedColor=<rgba 1 1 1 1>}
     uVRegion {bIsValid=false, max=<0 0>, min=<0 0>}
    }
   }
   slot {class=ScaleBoxSlot}
  }
  TutorialHighlight {widgetClassPath=&TutorialHighlight}
 ]
}
